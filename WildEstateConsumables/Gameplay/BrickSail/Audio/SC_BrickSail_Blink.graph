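# SoundCue for the BrickSail consumable's blink sound.
# A Random node picks one of three equally weighted blink wave variants, and a Modulator at the root randomises its pitch upward for variety.
# The cue plays quietly with a raised base pitch, using the weapon-drop attenuation and the loot-collect sound class.
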SoundCue /WildEstateConsumables/Gameplay/BrickSail/Audio/SC_BrickSail_Blink
root=Modulator_0
nodes [
 {id=Modulator_0, children [Random_0], properties {pitchMax=1.3, pitchMin=1}}
 {id=Random_0, children [WavePlayer_2,WavePlayer_4,WavePlayer_5], properties {weights [1,1,1]}}
 {id=WavePlayer_2, properties {soundWaveAssetPtr=&/BrickSail_Blink_01}}
 {id=WavePlayer_4, properties {soundWaveAssetPtr=&/BrickSail_Blink_02}}
 {id=WavePlayer_5, properties {soundWaveAssetPtr=&/BrickSail_Blink_03}}
]
properties {
 attenuationSettings=&Fort_Weapon_Drop_Att
 maxDistance=3000
 pitchMultiplier=1.2
 soundClassObject=&Fort_Loot_Collect
 volumeMultiplier=0.3
}
