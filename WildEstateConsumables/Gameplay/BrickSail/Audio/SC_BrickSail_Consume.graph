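# Sound cue that plays the consume sound for the BrickSail consumable.
# A Random node picks one of three consume wave variations with equal weight.
# Playback uses the shared consumable-use attenuation and the consumables sound class, with a slight volume boost.
# An overridden concurrency limit caps simultaneous instances and stops the oldest when exceeded.
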SoundCue /WildEstateConsumables/Gameplay/BrickSail/Audio/SC_BrickSail_Consume
root=Random_0
nodes [
 {id=Random_0, children [WavePlayer_1,WavePlayer_7,WavePlayer_8], properties {weights [1,1,1]}}
 {id=WavePlayer_1, properties {soundWaveAssetPtr=&/BrickSail_Consume_01}}
 {id=WavePlayer_7, properties {soundWaveAssetPtr=&/BrickSail_Consume_02}}
 {id=WavePlayer_8, properties {soundWaveAssetPtr=&/BrickSail_Consume_03}}
]
properties {
 attenuationSettings=&Fort_Consumable_Use_Athena_Att
 bOverrideConcurrency=true
 maxDistance=2250
 soundClassObject=&Fort_Consumables
 volumeMultiplier=1.1
 concurrencyOverrides {maxCount=4, resolutionRule=StopOldest}
}
