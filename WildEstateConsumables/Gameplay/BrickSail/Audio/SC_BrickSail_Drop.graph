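# Sound cue for dropping the BrickSail consumable item.
# A Random node picks one of three equally weighted drop waves and feeds a Modulator at the root.
# It uses the shared weapon-drop attenuation and loot-collect sound class, with a reduced volume multiplier.
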SoundCue /WildEstateConsumables/Gameplay/BrickSail/Audio/SC_BrickSail_Drop
root=Modulator_0
nodes [
 {id=Modulator_0, children [Random_0]}
 {id=Random_0, children [WavePlayer_0,WavePlayer_2,WavePlayer_3], properties {weights [1,1,1]}}
 {id=WavePlayer_0, properties {soundWaveAssetPtr=&/BrickSail_Drop_01}}
 {id=WavePlayer_2, properties {soundWaveAssetPtr=&/BrickSail_Drop_02}}
 {id=WavePlayer_3, properties {soundWaveAssetPtr=&/BrickSail_Drop_03}}
]
properties {
 attenuationSettings=&Fort_Weapon_Drop_Att
 maxDistance=3000
 soundClassObject=&Fort_Loot_Collect
 volumeMultiplier=0.45
}
