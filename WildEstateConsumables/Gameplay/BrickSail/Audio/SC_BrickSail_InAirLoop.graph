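# Looping in-air Sound Cue for the BrickSail consumable, reusing the PizzaParty slice in-air loop wave.
# A looping Wave Player feeds a default Modulator, with cue-level volume and pitch raised slightly.
# Inline overrides replace the referenced toy-hit attenuation with logarithmic falloff, and concurrency limits instances per owner with volume scaling.
# The cue plays under the Fort_Gadgets sound class.
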SoundCue /WildEstateConsumables/Gameplay/BrickSail/Audio/SC_BrickSail_InAirLoop
root=Modulator_1
nodes [
 {id=Modulator_1, children [WavePlayer_2]}
 {id=WavePlayer_2, properties {bLooping=true, soundWaveAssetPtr=&/PizzaParty_SliceInAirLoop_01}}
]
properties {
 attenuationSettings=/Game/Sounds/Attenuation/Fort_ToyHit_Att
 bOverrideAttenuation=true
 bOverrideConcurrency=true
 maxDistance=2000
 pitchMultiplier=1.2
 soundClassObject=&Fort_Gadgets
 volumeMultiplier=1.2
 attenuationOverrides {distanceAlgorithm=Logarithmic, falloffDistance=2000, stereoSpread=512, attenuationShapeExtents {x=0}}
 concurrencyOverrides {bLimitToOwner=true, maxCount=2, volumeScale=0.7}
}
