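# Sound cue for picking up the BrickSail consumable in the game.
# A Random node picks one of three pickup wave variations at equal weight and feeds a Modulator at its default settings.
# The cue plays quietly and pitched up, routed through the loot-collect sound class with shared weapon-pickup attenuation for distance falloff.
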SoundCue /WildEstateConsumables/Gameplay/BrickSail/Audio/SC_BrickSail_Pickup
root=Modulator_0
nodes [
 {id=Modulator_0, children [Random_0]}
 {id=Random_0, children [WavePlayer_1,WavePlayer_3,WavePlayer_4], properties {weights [1,1,1]}}
 {id=WavePlayer_1, properties {soundWaveAssetPtr=&/BrickSail_Pickup_01}}
 {id=WavePlayer_3, properties {soundWaveAssetPtr=&/BrickSail_Pickup_02}}
 {id=WavePlayer_4, properties {soundWaveAssetPtr=&/BrickSail_Pickup_03}}
]
properties {
 attenuationSettings=&Fort_Weapon_Pickup_Att
 maxDistance=3500
 pitchMultiplier=1.3
 soundClassObject=&Fort_Loot_Collect
 volumeMultiplier=0.18
}
