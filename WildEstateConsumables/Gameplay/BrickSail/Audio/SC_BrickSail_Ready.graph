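# Sound cue for the ready sound of the BrickSail consumable, reusing Flopper throw-ready waves.
# A Random node picks one of three equally weighted variations and feeds a default Modulator for pitch and volume variation.
# It plays at reduced volume in the gadget items sound class with gadget-placement attenuation.
# A concurrency override blocks new instances once the limit is reached instead of stopping playing ones.
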SoundCue /WildEstateConsumables/Gameplay/BrickSail/Audio/SC_BrickSail_Ready
root=Modulator_0
nodes [
 {id=Modulator_0, children [Random_0]}
 {id=Random_0, children [WavePlayer_1,WavePlayer_4,WavePlayer_5], properties {bShouldExcludeFromBranchCulling=true, weights [1,1,1]}}
 {id=WavePlayer_1, properties {soundWaveAssetPtr=&/Flopper_Throw_Ready_01}}
 {id=WavePlayer_4, properties {soundWaveAssetPtr=&/Flopper_Throw_Ready_02}}
 {id=WavePlayer_5, properties {soundWaveAssetPtr=&/Flopper_Throw_Ready_03}}
]
properties {
 attenuationSettings=/Game/Sounds/Attenuation/Fort_GadgetPlace_Att
 bOverrideConcurrency=true
 maxDistance=2000
 soundClassObject=&Fort_Gadgets_Items
 volumeMultiplier=0.5
 concurrencyOverrides {resolutionRule=PreventNew}
}
