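# Sound cue for the BrickSail spawn sound, part of the WildEstateConsumables gameplay content.
# A Random node picks one of three equally weighted spawn waves, which passes through a Modulator left at default settings.
# The cue raises pitch, lowers volume, uses the loot collect sound class and references weapon drop attenuation for distance falloff.
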SoundCue /WildEstateConsumables/Gameplay/BrickSail/Audio/SC_BrickSail_Spawn
root=Modulator_1
nodes [
 {id=Modulator_1, children [Random_3]}
 {id=Random_3, children [WavePlayer_5,WavePlayer_6,WavePlayer_7], properties {weights [1,1,1]}}
 {id=WavePlayer_5, properties {soundWaveAssetPtr=&/BrickSail_Spawn_01}}
 {id=WavePlayer_6, properties {soundWaveAssetPtr=&/BrickSail_Spawn_02}}
 {id=WavePlayer_7, properties {soundWaveAssetPtr=&/BrickSail_Spawn_03}}
]
properties {
 attenuationSettings=&Fort_Weapon_Drop_Att
 maxDistance=3000
 pitchMultiplier=1.3
 soundClassObject=&Fort_Loot_Collect
 volumeMultiplier=0.6
 attenuationOverrides {falloffDistance=2400, stereoSpread=2000, attenuationShapeExtents {x=800}}
}
