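# Sound cue for throwing the BrickSail consumable in WildEstateConsumables.
# A single wave player reuses the Flopper throw sample, passed through a Modulator at default settings for slight pitch and volume variation.
# It plays at reduced volume in the gadgets and items sound class, using gadget-placement attenuation.
# A concurrency override blocks new instances once the limit is reached instead of stopping playing ones.
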SoundCue /WildEstateConsumables/Gameplay/BrickSail/Audio/SC_BrickSail_Throw
root=Modulator_0
nodes [
 {id=Modulator_0, children [WavePlayer_0]}
 {id=WavePlayer_0, properties {soundWaveAssetPtr=&/Flopper_Throw_01}}
]
properties {
 attenuationSettings=/Game/Sounds/Attenuation/Fort_GadgetPlace_Att
 bOverrideConcurrency=true
 maxDistance=2000
 soundClassObject=&Fort_Gadgets_Items
 volumeMultiplier=0.8
 concurrencyOverrides {resolutionRule=PreventNew}
}
