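; Pickup Blueprint for the BrickSail consumable that gives the item a radioactive highlight look.
; On BeginPlay it runs the parent logic, then enables custom depth with a stencil value on the skeletal or static mesh primitive to drive the post effect.
; It also casts that primitive to a mesh component and applies the Buzzcool overlay material.
; The trailing next-tick delay leads nowhere, and the overlap and tick events are empty.
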(event EventBeginPlay
  (bind _skeletalorstaticmeshassetprimitive (Variables|Default|GetSkeletalOrStaticMeshAssetPrimitive))
  (|CallParentFunction)
  ; Addition so the pickup exhibits the radioactive post effect.
  (Rendering|SetRenderCustomDepth _skeletalorstaticmeshassetprimitive true)
  (Rendering|SetCustomDepthStencilValue _skeletalorstaticmeshassetprimitive 5)
  (bind _asmesh_component (Utilities|Casting|CastToMeshComponent _skeletalorstaticmeshassetprimitive))
  (Rendering|Material|SetOverlayMaterial _asmesh_component "/WildEstateTech/Gameplay/Materials/MI_Buzzcool_Overlay.MI_Buzzcool_Overlay" false 0 :!class /Script/Engine.MeshComponent)
  (Utilities|FlowControl|DelayUntilNextTick))

(event Collision|EventActorBeginOverlap
  (params (OtherActor :type Actor%20Object%20Reference)))

(event EventTick
  (params (DeltaSeconds :type Float%20%28single-precision%29)))

(fn ConstructionScript
  (|CallParentFunction))
