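; Thrown projectile for the BrickSail consumable that adds an in-air sound loop and reports throws into water.
; The loop plays on BeginPlay and stops on hit, with parent logic preserved.
; When its pickup spawns, it multi-traces from its location along a water offset.
; A blocking hit broadcasts the ThrowWater custom verb for the instigator's pawn; the trace has timed debug drawing enabled.
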(event EventBeginPlay
  (|CallParentFunction)
  (Audio|Components|Audio|Play (Variables|Default|GetAudioInAirLoop) 0.0))

(event EventOnPickupSpawned
  (params (Spawned%20Pickup :type Fort%20Pickup%20Object%20Reference))
  (|CallParentFunction Spawned%20Pickup)
  (bind (_outhits _returnvalue) (Collision|MultiLineTraceByChannel :Start (Transformation|GetActorLocation) :End (+ (Transformation|GetActorLocation) (Variables|Default|GetWaterTraceOffset)) :TraceChannel "TraceTypeQuery12" :bTraceComplex false :DrawDebugType "ForDuration" :bIgnoreSelf true :TraceColor "(R=1.000000,G=0.142595,B=0.757396,A=1.000000)" :TraceHitColor "(R=0.165754,G=1.000000,B=0.121057,A=1.000000)" :DrawTime 1e+01))
  (if _returnvalue
    (CustomVerb|BroadcastCustomVerbMessage (CustomVerb|MakeCustomVerbMessage (Game|Player|GetPlayerPawn (Projectile|GetInstigatorPlayerController :!class /Script/FortniteGame.FortProjectileMovementInterface)) "(TagName=\"CustomVerb.BrickSail.ThrowWater\")"))))

(event Collision|EventHit
  (params (MyComp :type Primitive%20Component%20Object%20Reference) (Other :type Actor%20Object%20Reference) (OtherComp :type Primitive%20Component%20Object%20Reference) (bSelfMoved :type Boolean) (HitLocation :type Vector :default "0, 0, 0") (HitNormal :type Vector :default "0, 0, 0") (NormalImpulse :type Vector :default "0, 0, 0") (Hit :type Hit%20Result%20Structure%20%28by%20ref%29))
  (Audio|Components|Audio|Stop (Variables|Default|GetAudioInAirLoop))
  (|CallParentFunction MyComp Other OtherComp bSelfMoved HitLocation HitNormal NormalImpulse Hit))

(fn ConstructionScript
  (|CallParentFunction))
