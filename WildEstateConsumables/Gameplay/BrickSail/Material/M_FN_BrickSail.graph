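# BrickSail character material: 2023 anime shader with procedural eye blinking and optional emissive glow, flagged for skeletal meshes.
# UV1 gradients separate left and middle eyes for offset timing, while time-driven sines move a hard-edged eyelid mask, with debug scrubbing.
# That mask feeds a BrickSail-specific Pre function, chained into shared anime Calculations and Post functions.
# A static switch adds colored, time-oscillating Fresnel glow to the resulting emissive attribute.
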Material /WildEstateConsumables/Gameplay/BrickSail/Material/M_FN_BrickSail
graph {
 properties {
  bUseMaterialAttributes=true
  bUsedWithMeshDeformer=true
  bUsedWithNanite=true
  bUsedWithSkeletalMesh=true
  pixelDepthOffsetMode=PDOM_Legacy
  usedShadingModels=Default Lit
 }
 nodes [
  n1 = NamedRerouteUsage()
  n2 = MaterialFunctionCall(UV Channel=TextureCoordinate() {coordinateIndex=1}) {materialFunction=&LinearGradient}
  n3 = ScalarParameter() {defaultValue=0.63, group=Blink, parameterName=left eye mask}
  n4 = Saturate(None=Multiply(A=Subtract(A=@n2.UGradient, B=@n3)) {constB=256})
  NamedRerouteDeclaration(None=Multiply(A=@n1, B=@n4)) {name=left mask, nodeColor=<rgba 0.5703 0 1 1>}
  n5 = MaterialFunctionCall(UV Channel=TextureCoordinate() {coordinateIndex=1}) {materialFunction=&LinearGradient}
  n6 = ScalarParameter() {defaultValue=0.35, group=Blink, parameterName=Middle eye mask}
  n7 = Saturate(None=Multiply(A=Subtract(A=@n5.UGradient, B=@n6)) {constB=256})
  NamedRerouteDeclaration(None=Multiply(A=Subtract(A=@n7, B=@n4), B=@n1)) {name=middle mask, nodeColor=<rgba 0 1 0.25 1>}
  n8 = Time()
  n9 = Ceil(None=Sine(None=Multiply(A=@n8) {constB=0.2}) {period=4})
  NamedRerouteDeclaration(None=StaticSwitchParameter(True=Constant() {r=1}, False=LinearInterpolate(Alpha=@n9) {constA=1, constB=0}) {group=Blink, parameterName=Debug_Eyes}) {name=Blink Type Mask, nodeColor=<rgba 0.2422 0 1 1>}
  PinBase()
  n10 = DotProduct(A=PixelNormalWS(), B=CameraVectorWS())
  n11 = Power(Base=@n10, Exp=ScalarParameter() {defaultValue=2, group=Global, parameterName=PowA_Fresnal})
  n12 = Power(Base=@n10, Exp=ScalarParameter() {defaultValue=3.25, group=Global, parameterName=PowB_Fresnal}) {constExponent=3.25}
  n13 = MaterialFunctionCall(Speed=Constant() {r=0.5}) {materialFunction=&DebugTimeSine}
  n14 = Saturate(None=OneMinus(None=LinearInterpolate(A=@n11, B=@n12, Alpha=@n13.Result)))
  PinBase!(Pin 1=@n14) {reroutePins [Pin 1 {expression=&self:MaterialExpressionReroute_6}]}
  MaterialExpressionReroute_6 = Reroute(None=@n14)
  n15 = TextureCoordinate()
  n16 = ComponentMask(None=VectorParameter() {parameterName="EyeMaskPosition(Do Not Touch)", defaultValue=<rgba 0.5 -0.5 0 1>}.RGB) {g=true, r=true}
  n17 = ScalarParameter() {defaultValue=0.1, parameterName="X_EyeMask_Size(Do Not Touch)"}
  n18 = ScalarParameter() {defaultValue=0.1, parameterName="Y_EyeMask_Size(Do Not Touch)"}
  n19 = Clamp(None=MaterialFunctionCall(UVs=Add(A=@n15, B=@n16), Texture Scale=AppendVector(A=@n17, B=@n18)) {materialFunction=&ScaleUVsByCenter}.0-1 mask)
  n20 = StaticSwitchParameter(True=ScalarParameter() {group=Blink, parameterName=debug_Blink_percentage}, False=Time()) {group=Blink, parameterName=Debug_Eyes}
  n21 = LinearInterpolate(A=@n20, B=Add(A=@n20, B=ScalarParameter() {group=Blink, parameterName=left_eye_Offset}), Alpha=NamedRerouteUsage())
  n22 = Add(A=@n21, B=ScalarParameter() {group=Blink, parameterName=middle_eye_Offset})
  n23 = NamedRerouteUsage()
  n24 = Sine(None=Multiply(A=LinearInterpolate(A=@n21, B=@n22, Alpha=@n23)))
  n25 = LinearInterpolate(A=@n20, B=Add(A=@n20, B=ScalarParameter() {group=Blink, parameterName=left_eye_Offset}), Alpha=NamedRerouteUsage())
  n26 = Add(A=@n25, B=ScalarParameter() {group=Blink, parameterName=middle_eye_Offset})
  n27 = NamedRerouteUsage()
  n28 = Sine(None=Multiply(A=LinearInterpolate(A=@n25, B=@n26, Alpha=@n27)) {constB=0.2})
  n29 = Subtract(
   A=MaterialFunctionCall(UV Channel=TextureCoordinate() {coordinateIndex=1}) {materialFunction=&LinearGradient}.VGradient
   B=Multiply(A=ConstantBiasScale(None=@n24), B=ConstantBiasScale(None=@n28) {bias=0.5})
  )
  n30 = Saturate(None=OneMinus(None=Multiply(A=@n29) {constB=256}))
  n31 = MaterialFunctionCall!(Eyelid_Mask=NamedRerouteDeclaration(None=Multiply(A=@n19, B=@n30)) {name=eyelid mask, nodeColor=<rgba 0.6406 0 1 1>}, CustomUV) {materialFunction=&/MFC_AnimeShader_2023_Pre_Bricksail}
  n32 = MaterialFunctionCall!(CustomLightVector, OverrideNormal_WS, STT=@n31.STT, Normal_Map=@n31.STT, DFL=@n31.STT, FX=@n31.STT) {materialFunction=/Game/Global/Materials/Character_Parent_2023/Anime_Parent_2023/MaterialFunctions/MFC_AnimeShader_2023_Calculations}
  n33 = MaterialFunctionCall!(
   Shading_Gradient=@n32.Shading_Gradient
   Threshold_Gradient=@n32.Shading_Gradient
   Specular_Gradient=@n32.Shading_Gradient
   Sheen_Mask=@n32.Shading_Gradient
   Thickness=@n32.Shading_Gradient
   DFL_Mask=@n32.Shading_Gradient
   FX=@n32.Shading_Gradient
   Metallic_Gradient=@n32.Shading_Gradient
   RimLight_Gradient=@n32.Shading_Gradient
   NormalBuffer=@n32.Shading_Gradient
   --------
   CINE_LightSources=@n32.Shading_Gradient
   CINE_LightColor=@n32.Shading_Gradient
   ---------
   Color_Lit=@n31.Color_Lit
   Color_Shaded=@n31.Color_Lit
   SRM=@n31.Color_Lit
   Emissive=@n31.Color_Lit
   DFL_Color=@n31.Color_Lit
   Specular_Color=@n31.Color_Lit
   RimLight_Color=@n31.Color_Lit
  ) {materialFunction=/Game/Global/Materials/Character_Parent_2023/Anime_Parent_2023/MaterialFunctions/MFC_AnimeShader_2023_Post}
  n34 = GetMaterialAttributes(None=@n33.Result) {attributeGetTypes [#1]}
  n35 = DotProduct(A=PixelNormalWS(), B=CameraVectorWS())
  n36 = LinearInterpolate(
   A=Power(Base=@n35, Exp=ScalarParameter() {defaultValue=2, group=Global, parameterName=PowA_Fresnal})
   B=Power(Base=@n35, Exp=ScalarParameter() {defaultValue=3.25, group=Global, parameterName=PowB_Fresnal}) {constExponent=3.25}
   Alpha=MaterialFunctionCall(Speed=Constant() {r=0.5}) {materialFunction=&DebugTimeSine}.Result
  )
  n37 = Multiply(A=VectorParameter() {group=Global, parameterName=EmissiveColor, defaultValue=<rgba 0 0 0 1>}.RGB, B=Saturate(None=OneMinus(None=@n36)))
  StaticSwitchParameter(
   True=SetMaterialAttributes(MaterialAttributes=@n34.MaterialAttributes, Emissive Color=Add(A=@n34.Emissive Color, B=@n37)) {attributeSetTypes [#1]}
   False=@n33.Result
  ) {parameterName=Use_GlowEmissive}
  VertexInterpolator(VS=PreSkinnedPosition())
 ]
}
