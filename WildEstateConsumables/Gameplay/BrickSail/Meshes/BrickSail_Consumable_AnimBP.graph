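# Animation Blueprint for the BrickSail consumable's skeleton, adding procedural spine and tail motion plus an adjustable offset.
# The reference pose runs through front and back spine Trail controllers, the latter weighted by TailAlpha, then into DefaultSlot.
# The slot output gets an additive bone-space translation on the Anim bone, driven by Y_Offset and OffsetAlpha.
# The EventGraph is empty, those variables are not declared here, and a front-spine Spring controller is left unconnected.
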AnimBlueprint /WildEstateConsumables/Gameplay/BrickSail/Meshes/BrickSail_Consumable_AnimBP
target_skeleton=&/BrickSail_Consumable_Skeleton
graphs [
 AnimGraph {nodes [
   n1 = Trail controller - Bone: dyn_spineFront_02!(ComponentPose=Local To Component(LocalPose=Local Space Ref Pose()), Alpha=0.2, RelaxationSpeedScale:float=4) -> (Pose) {
    maxDeltaTime=0.0333
    relaxationSpeedScale=4
    rotationLimits [{limitMax=<180 5 180>, limitMin=<-180 -5 -180>}, {limitMax=<180 5 180>, limitMin=<-180 -5 -180>}]
    rotationOffsets [<0 0 0>, <0 0 0>]
    trailBone {boneName=dyn_spineFront_02}
   }
   n2 = Trail controller - Bone: dyn_spineBack_02!(ComponentPose=@n1, Alpha=Get TailAlpha() -> (TailAlpha:float), RelaxationSpeedScale:float=4) -> (Pose) {
    maxDeltaTime=0.0333
    relaxationSpeedScale=3
    rotationLimits [{limitMax=<5 15 65>, limitMin=<-5 -15 -65>}, {limitMax=<5 15 65>, limitMin=<-5 -15 -65>}]
    rotationOffsets [<0 0 0>, <0 0 0>]
    trailBone {boneName=dyn_spineBack_02}
   }
   n3 = Component To Local(ComponentPose=@n2)
   n4 = "Transform (Modify) Bone - Bone: Anim"!(
    ComponentPose=Local To Component(LocalPose=Slot 'DefaultSlot'(Source=@n3))
    Alpha=Get OffsetAlpha() -> (OffsetAlpha:float)
    Translation_X:double=0
    Translation_Y:double=Get Y_Offset() -> (Y_Offset:float)
    Translation_Z:double=2
   ) -> (Pose) {translationMode=BMM_Additive, translationSpace=BCS_BoneSpace, boneToModify {boneName=Anim}, translation=<-35 8 20>}
   Output Pose(Result=Component To Local(ComponentPose=@n4)) {layerGroup=None}
   Spring controller - Bone: dyn_spineFront_01!(ComponentPose:Component Space Pose Link Structure, Alpha:float=1) -> (Pose:Component Space Pose Link Structure) {
    is=title AnimGraphNode_SpringBone
    bLimitDisplacement=true
    maxDisplacement=1
    alphaBoolBlend {blendOption=Linear}
    springBone {boneName=dyn_spineFront_01}
   }
  ]}
 EventGraph {links [], nodes []}
]
variables []
