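# Niagara effect for the energy drink consumable that tosses a discarded bottle as a short-lived mesh particle.
# A single burst emitter launches the mesh backward along X and downward, where gravity and bouncing collision take over while a curve-driven spin slows.
# Mesh scale pops in and shrinks out over the lifetime, and a time-of-day function lowers color alpha in evening and at night.
# A curve-based mesh rotation force is present but disabled.
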NiagaraSystem /WildEstateConsumables/Gameplay/BrickSail/VFX/NS_Consumable_EnergyDrink_BottleDiscard
dependencies {
 usedDataInterfaces [&NiagaraDataInterfaceCurve, &NiagaraDataInterfaceVectorCurve]
 usedDynamicInputs []
 usedModules [&SystemState]
}
emitters {NE_Bottle {
  emitterUpdateScript [
   &EmitterState()
   &SpawnBurst_Instantaneous(Use Spawn Probability=false, UseLoopCountLimit=false)
  ]
  particleSpawnScript [
   &InitializeParticle(Color=<rgba 0.6058 0.6058 0.6058 3>, Lifetime=0.5, Mesh Scale Mode=::NewEnumerator3, UsePositionOffset=false, Write Material Random=true)
   &InitialMeshOrientation(Mesh Axis To Orient=<0 0 1>, Mesh Orientation Mode=::NewEnumerator2, Orientation Vector=<0 0 1>, Use Rotation Vector=false)
   &AddVelocity(Velocity=<-450 0 -200>)
   /Niagara/Modules/Update/Mass/CalculateSizeAndRotationalInertiaByMass(Depth=5, Height=2, Module.Manually Enter Density=false, Width=5)
  ]
  particleUpdateScript [
   &ParticleState()
   /Niagara/Modules/Update/Forces/FindKineticAndPotentialEnergy()
   &ScaleMeshSize(Scale Factor=&VectorFromFloat(Value=&FloatFromCurve(FloatCurve=DataInterface() {propertyValues {Curve {keys [
          {interpMode=RCIM_Cubic, tangentMode=RCTM_User, time=0, value=0}
          {interpMode=RCIM_Cubic, tangentMode=RCTM_User, time=0.1, value=1}
          {arriveTangent=-1.1111, interpMode=RCIM_Cubic, leaveTangent=-1.1111, time=0.8, value=1}
          {tangentMode=RCTM_User, time=1, value=0}
         ]}}})))
   &GravityForce(Gravity=<0 0 -1500>)
   &UpdateMeshOrientation(Rotation Rate=&FloatFromCurve(FloatCurve=DataInterface() {propertyValues {Curve {keys [<at 0.25 1>,<at 1 0>]}}}, Scale Curve=25))
   &Collision(
    Age Colliding Particles=false
    CPU Collision Trace Channel=ECC_Pawn
    Control Roll On Collision=true
    Mesh Dimensions:Vector3f=<15 15 5>
    Method for Calculating Particle Radius:ENiagaraCalculateRadiusOptions=::NewEnumerator1
    Module.Control Rest State Via Surface Normal=false
    Module.EnableMaxCollisionCount=false
    Module.Randomize Collision Normal Vector=false
    Radius Calculation Type=::NewEnumerator1
    Restitution=0.5
    ScaleRateOfRotation=false
   )
   &MeshRotationForce(Rotation=&VectorFromCurve(
     VectorCurve=DataInterface() {propertyValues {XCurve {keys [<at 0 1>,<at 1 0>]}, YCurve {keys [<at 0 1>,<at 1 0>]}, ZCurve {keys [<at 0 1>,<at 1 0>]}}}
     Scale Curve=<5 5 5>
    )) {enabled=false}
   &SolveForcesAndVelocity(Clamp Velocity=false, Limit Acceleration=false)
   &ScaleColor(
    Scale Alpha=&FNTimeOfDayScalar(Night Scale=0.3, Morning Scale=1, Evening Scale=0.3)
    ScaleA=true
    ScaleRGB=true
   )
  ]
  rendererClasses [&NiagaraMeshRendererProperties]
  renderers [{rendererClass=&NiagaraMeshRendererProperties, rendererIndex=0}]
 }}
summary {
 systemName=NS_Consumable_EnergyDrink_BottleDiscard
 emitters [{emitterName=NE_Bottle, rendererClasses [&NiagaraMeshRendererProperties]}]
 userVariables []
}
